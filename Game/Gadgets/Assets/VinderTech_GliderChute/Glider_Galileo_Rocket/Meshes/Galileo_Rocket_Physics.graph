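# Physics asset for the Galileo Rocket glider, a VinderTech GliderChute gadget mesh.
# A single box shape on the Body_Offset bone gives the whole glider one simplified collision volume, offset from the bone origin.
# It has no constraints, so the asset defines no jointed or multi-body ragdoll behavior.
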PhysicsAsset /Game/Gadgets/Assets/VinderTech_GliderChute/Glider_Galileo_Rocket/Meshes/Galileo_Rocket_Physics
bodies {Body_Offset [{shapeType=Box, extentX=327.6875, extentY=462.75, extentZ=75.25, center=<-0.0002 -70.7833 28.8284>}]}
constraints []
